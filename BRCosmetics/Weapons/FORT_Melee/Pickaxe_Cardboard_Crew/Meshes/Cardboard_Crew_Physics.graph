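# Physics asset for the Cardboard Crew pickaxe, a two-handed melee cosmetic, giving its cardboard pieces jointed, limited motion.
# A capsule on the handle anchors left and right cardboard capsules through joints with narrow swing and twist limits.
# Each cardboard capsule carries two box-shaped boards whose joints lock twist and one swing axis, so each board flaps like a hinge on the remaining axis.
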PhysicsAsset /BRCosmetics/Weapons/FORT_Melee/Pickaxe_Cardboard_Crew/Meshes/Cardboard_Crew_Physics
bodies {
 Melee_TwoHanded_Handle [{shapeType=Capsule, length=101.1076, radius=4.2511, center=<1.3886 -0.8657 32.4177>}]
 dyn_board_01_l [{shapeType=Box, extentX=12, extentY=25, extentZ=3, center=<4 0 0>}]
 dyn_board_01_r [{shapeType=Box, extentX=12, extentY=25, extentZ=3, center=<4 0 0>}]
 dyn_board_02_l [{shapeType=Box, extentX=6, extentY=25, extentZ=3, center=<3 0 0>}]
 dyn_board_02_r [{shapeType=Box, extentX=12, extentY=25, extentZ=3, center=<4 0 0>}]
 dyn_cardboard_l [{shapeType=Capsule, length=5, radius=6, center=<2 0 0>, rotation=<rot 90 0 0>}]
 dyn_cardboard_r [{shapeType=Capsule, length=5, radius=6, center=<2 0 0>, rotation=<rot 90 0 0>}]
}
constraints [
 {bone1Name=dyn_cardboard_l, bone2Name=Melee_TwoHanded_Handle, swing1LimitDegrees=20, swing2LimitDegrees=20, twistLimitDegrees=15}
 {bone1Name=dyn_cardboard_r, bone2Name=Melee_TwoHanded_Handle, swing1LimitDegrees=20, swing2LimitDegrees=20, twistLimitDegrees=15}
 {
  bone1Name=dyn_board_01_l
  bone2Name=dyn_cardboard_l
  swing1LimitDegrees=45
  swing1Motion=Locked
  swing2LimitDegrees=60
  twistLimitDegrees=45
  twistMotion=Locked
 }
 {
  bone1Name=dyn_board_02_l
  bone2Name=dyn_cardboard_l
  swing1LimitDegrees=45
  swing1Motion=Locked
  swing2LimitDegrees=60
  twistLimitDegrees=45
  twistMotion=Locked
 }
 {
  bone1Name=dyn_board_01_r
  bone2Name=dyn_cardboard_r
  swing1LimitDegrees=45
  swing1Motion=Locked
  swing2LimitDegrees=60
  twistLimitDegrees=45
  twistMotion=Locked
 }
 {
  bone1Name=dyn_board_02_r
  bone2Name=dyn_cardboard_r
  swing1LimitDegrees=45
  swing1Motion=Locked
  swing2LimitDegrees=60
  twistLimitDegrees=45
  twistMotion=Locked
 }
]
