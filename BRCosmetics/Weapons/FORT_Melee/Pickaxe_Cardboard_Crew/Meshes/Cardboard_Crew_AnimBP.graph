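# Animation Blueprint for the Cardboard Crew pickaxe mesh that adds secondary motion and shows damage as its boxes break.
# A RigidBody node simulates the override physics asset relative to the two-handed handle bone, with scaled inertia and reduced gravity.
# Each update reads the impact actor's break flags: BreakBoxA toggles a material hide parameter, BreakBoxB sets all_busted.
# When all_busted is true, three chained additive bone rotations bend the tube after simulation.
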AnimBlueprint /BRCosmetics/Weapons/FORT_Melee/Pickaxe_Cardboard_Crew/Meshes/Cardboard_Crew_AnimBP
target_skeleton=&/Cardboard_Crew_Skeleton
graphs [
 AnimGraph {nodes [
   AnimGraphNode_SubInput_0 = Input Pose()
   n1 = RigidBody(ComponentPose=Local To Component(LocalPose=@AnimGraphNode_SubInput_0)) {
    bClampLinearTranslationLimitToRefPose=true
    bOverrideWorldGravity=true
    overridePhysicsAsset=&/Cardboard_Crew_Physics
    simulationSpace=BaseBoneSpace
    alphaBoolBlend {blendOption=Linear}
    baseBoneRef {boneName=Melee_TwoHanded_Handle}
    componentAppliedLinearAccClamp=<2000 2000 2000>
    componentLinearAccScale=<0.2 0.2 0.2>
    componentLinearVelScale=<0.15 0.15 0.15>
    overrideWorldGravity {z=-300}
   }
   n2 = Get all_busted() -> (all_busted:bool)
   n3 = "Transform (Modify) Bone - Bone: dyn_tube_01"!(ComponentPose=@n1, bAlphaBoolEnabled:bool=@n2, Translation="0,0,0", Rotation="-27,-16.760,7.250", Scale="1,1,1") -> (Pose) {
    alphaInputType=Bool
    rotationMode=BMM_Additive
    rotationSpace=BCS_BoneSpace
    alphaBoolBlend {blendOption=Linear}
    boneToModify {boneName=dyn_tube_01}
    rotation=<rot -27 -16.76 7.25>
   }
   n4 = "Transform (Modify) Bone - Bone: dyn_tube_02"!(ComponentPose=@n3, bAlphaBoolEnabled:bool=@n2, Translation="0,0,0", Rotation="-34,-17.941,7.856", Scale="1,1,1") -> (Pose) {
    alphaInputType=Bool
    rotationMode=BMM_Additive
    rotationSpace=BCS_BoneSpace
    alphaBoolBlend {blendOption=Linear}
    boneToModify {boneName=dyn_tube_02}
    rotation=<rot -34 -17.941 7.856>
   }
   n5 = "Transform (Modify) Bone - Bone: dyn_tube_03"!(ComponentPose=@n4, bAlphaBoolEnabled:bool=@n2, Translation="0,0,0", Rotation="-0.806,-20.959,6.616", Scale="1,1,1") -> (Pose) {
    alphaInputType=Bool
    rotationMode=BMM_Additive
    rotationSpace=BCS_BoneSpace
    alphaBoolBlend {blendOption=Linear}
    boneToModify {boneName=dyn_tube_03}
    rotation=<rot -0.806 -20.9587 6.6161>
   }
   Output Pose(Result=Component To Local(ComponentPose=@n5)) {layerGroup=None}
  ]}
 EventGraph {nodes [
   n1 = GetOwningComponent()
   n2 = Get Dynamic Material Instance(SkeletalMeshComponent=@n1)
   > SetScalarParameterValue(self=@n2, ParameterName=Hide Element 01, Value=0)
   n3 = Get Dynamic Material Instance(SkeletalMeshComponent=@n1)
   > SetScalarParameterValue(self=@n3, ParameterName=Hide Element 01, Value=1)
   Event BlueprintUpdateAnimation()
   > n4 = Cast To B_Melee_Impact_Pickaxe_Athena_Cardboard_Crew!(execute, Object=GetOwningActor()) -> (then, CastFailed, AsB Melee Impact Pickaxe Athena Cardboard Crew:B Melee Impact Pickaxe Athena Cardboard Crew Object Reference) {is=title K2Node_DynamicCast}
   > n5 = Sequence()
   > n6 = Branch(Condition=Get BreakBoxA!(self:B Melee Impact Pickaxe Athena Cardboard Crew Object Reference=@n4) -> (BreakBoxA:bool))
   n7 = GetOwningComponent()
   > n8 = Get Dynamic Material Instance(SkeletalMeshComponent=@n7)
   > SetScalarParameterValue(self=@n8, ParameterName=Hide Element 02, Value=1)
   n9 = Get Dynamic Material Instance(execute=@n6.else, SkeletalMeshComponent=@n7)
   > SetScalarParameterValue(self=@n9, ParameterName=Hide Element 02, Value=0)
   n10 = Branch(execute=@n5.then_1, Condition=Get BreakBoxB!(self:B Melee Impact Pickaxe Athena Cardboard Crew Object Reference=@n4) -> (BreakBoxB:bool))
   > Set all_busted(all_busted:bool=true) -> (Output_Get:bool)
   Set all_busted(execute=@n10.else, all_busted=false)
   TryGetPawnOwner()
  ]}
]
variables [all_busted {type=boolean}]
